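; Placement preview Blueprint for the player's Teleporter gadget, showing a reticle effect and playing audio while the gadget is being placed.
; Visual updates branch on whether the spot is safe, using DoOnce-gated template swaps on the Good particle component between the teleporter reticle and a bad-spot reticle borrowed from SupplyDrop.
; Placement audio starts on BeginPlay and stops when a scalar parameter is set on the preview materials; the other preview events are empty.
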(event PlacementVisuals|EventSetVectorParameterValueOnAllPreviewMIDs
  (params (ParamName :type Name :default "None") (ParamValue :type Linear%20Color%20Structure%20%28by%20ref%29)))

(event PlacementVisuals|EventOnBounceMotionUpdate
  (params (VectorScale :type Float%20%28single-precision%29)))

(event EventBeginPlay
  (Audio|Components|Audio|Play (Variables|Default|GetPlacement_Audio) 0.0))

(event PlacementVisuals|EventOnInitializeDecoPreview
  (params (NewBuildingActor :type Building%20Actor%20Object%20Reference) (NewDecoItemDefinition :type Fort%20Deco%20Item%20Definition%20Object%20Reference)))

(event PlacementVisuals|EventSetScalarParameterValueOnAllPreviewMIDs
  (params (ParamName :type Name :default "None") (ParamValue :type Float%20%28single-precision%29))
  (Audio|Components|Audio|Stop (Variables|Default|GetPlacement_Audio)))

(event PlacementVisuals|EventOnUpdateScale
  (params (Scale :type Float%20%28single-precision%29) (InWorldSpacePivot :type Vector%20%28by%20ref%29 :default "0, 0, 0")))

(event PlacementVisuals|EventOnUpdateVisuals
  (params (DeltaSeconds :type Float%20%28single-precision%29) (bSafeToPlace :type Boolean))
  (bind _good (Variables|Default|GetGood))
  (if bSafeToPlace
    (Utilities|FlowControl|Sequence
      (:then_0
        _)
      (:then_1
        (Utilities|FlowControl|DoOnce false
          (:Completed
            (Effects|Components|ParticleSystem|SetTemplate _good "/Game/Gadgets/Assets/Teleporter/Effects/P_Teleporter_Reticle.P_Teleporter_Reticle")))))
    (else
      (Utilities|FlowControl|Sequence
        (:then_0
          _)
        (:then_1
          _))))
  (Utilities|FlowControl|DoOnce false
    (:Completed
      (Effects|Components|ParticleSystem|SetTemplate _good "/Game/Gadgets/Assets/SupplyDrop/Effects/P_SupplyDrop_ReticleBadSpot.P_SupplyDrop_ReticleBadSpot"))))

(event EventDestroyed)

(fn ConstructionScript)
