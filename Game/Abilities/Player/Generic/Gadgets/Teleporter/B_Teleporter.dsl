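; B_Teleporter is a carriable player gadget that links to a target teleporter and moves overlapping pawns, vehicles, and optionally husks or projectiles to it, redirecting their velocity.
; The server gates teleports on tags, active state, recharge, carrying and range, using bTeleporting against bounce-back.
; Ground-actor events and traces relocate it when its floor dies; destruction also destroys its pair.
; OnRep functions and timelines drive glow, recharge ring, arrow and audio.
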(event EventOnWorldReady
  ;[1 1 1] BeginPlay
  (Collision|AssignOnComponentBeginOverlap (Variables|Default|GetPortalCollisionCylinder) (delegate Custom|OnComponentBeginOverlap_Event_0))
  (Networking|CanExecuteCosmeticEvents
    (:True
      (Components|Animation|PlayAnimation (Variables|Default|GetSK_Teleporter) "/Game/Gadgets/Assets/Teleporter/SK_Teleporter_Anim.SK_Teleporter_Anim" false)
      (CallFunction|DirectionalRayActivate)
      (Visualization|UpdateGlowVisuals)
      (CallFunction|UpdateArrowDirection)
      ;[1 1 1] Make arrows on each teleporter rotate and face one another
      (bind _returnvalue (Utilities|Time|SetTimerbyEvent (delegate Custom|UpdateArrowDirection) 0.2 true))
      (Variables|Default|SetUpdateArrowDirectionTimerHandle _returnvalue))
    (:False
      (CallFunction|FirstPlaceActivation))))

(event Custom|UpdateArrowDirection
  ;[1 1 1] Make arrows on each teleporter rotate and face one another
  (Variables|VFX|SetArrow%20Rotation%20New (CallFunction|GetArrowRotation))
  (if (not (Math|Rotator|Equal(Rotator) (Variables|VFX|GetArrow%20Rotation%20New) (Variables|VFX|GetArrow%20Rotation%20Previous) 0.5))
    (bind (_direction _lerp) (Timeline|RotateArrowTimeline PlayFromStart)
      (:Update
        (Transformation|SetWorldRotation (Variables|Default|GetDirectionalArrow) (Math|Rotator|Lerp(Rotator) (Variables|VFX|GetArrow%20Rotation%20Previous) (Variables|VFX|GetArrow%20Rotation%20New) _lerp true) false false))
      (:Finished
        (Variables|VFX|SetArrow%20Rotation%20Previous (Variables|VFX|GetArrow%20Rotation%20New))))))

(event Custom|SetRemoteCarried
  (Variables|Default|SetRemoteCarried true)
  (BusinessLogic|UpdateTeleportersActive))

(event Custom|OnComponentBeginOverlap_Event_0
  (params (OverlappedComponent :type Primitive%20Component%20Object%20Reference) (OtherActor :type Actor%20Object%20Reference) (OtherComp :type Primitive%20Component%20Object%20Reference) (OtherBodyIndex :type Integer) (bFromSweep :type Boolean) (SweepResult :type Hit%20Result%20Structure%20%28by%20ref%29))
  (bind _actorstoteleport (Variables|BusinessLogic|GetActorsToTeleport))
  ;[1 1 1] Touched Teleporter
  (Networking|SwitchHasAuthority
    (:Authority
      (if (Math|Boolean|ANDBoolean (not (Variables|BusinessLogic|GetbTeleporting)) (not (Variables|Replicated|GetbBeingCarried)) (not (Variables|Default|GetRemoteCarried)))
        ;[1 0.3541 0.364] Changing to hopefully only work on Players and not AI
        (bind (_teleportactorabilitysystem _validactor) (CallFunction|GateTeleportedActorType OtherActor (Variables|DesignRestrictions|GetbTeleportHusks) (Variables|DesignRestrictions|GetbTeleportProjectiles) (Collision|GetCollisionObjectType OtherComp)))
        (if _validactor
          (Utilities|IsValid _teleportactorabilitysystem
            (:Is%20Valid
              (if (not (GameplayTags|HasMatchingGameplayTag _teleportactorabilitysystem "(TagName=\"Abilities.Generic.Teleporter.Immune\")"))
                _))
            (:Is%20Not%20Valid
              _))
          ;;[1 1 1] Defer teleportation to avoid initialization issues
          (Utilities|Array|Add (Variables|BusinessLogic|GetActorsToTeleport) OtherActor)
          (for _array_element _actorstoteleport
            (BusinessLogic|TeleporttoTarget _array_element))
          (Utilities|Array|Clear _actorstoteleport)
          (Utilities|FlowControl|Delay 0.2)
          (Variables|BusinessLogic|SetbTeleporting false (Variables|BusinessLogic|GetTargetTeleporter)))))))

(event Carry|EventOnDropped
  (params (bValidLocation :type Boolean) (bPassesPlacementRestrictions :type Boolean) (bWasIntentionallyPlaced :type Boolean))
  ;[1 1 1] Drop
  (Variables|BusinessLogic|SetbIntentionalPlacement bWasIntentionallyPlaced)
  (Variables|BusinessLogic|SetbValidDropLocation (and bValidLocation bPassesPlacementRestrictions))
  (BusinessLogic|TriggerRecharge) ; Triggers Recharge for both teleporters
  (CallFunction|UnbindWeaponSwap)
  (Variables|Replicated|SetbBeingCarried false)
  (CallFunction|BindFlooronDestroyed)
  (CallFunction|CheckGroundPosition false)
  (BusinessLogic|UpdateTeleportersActive)
  (BusinessLogic|DestroyifSweepAgainstKillZVolume (Transformation|GetActorLocation)))

(event Interaction|EventBlueprintOnInteract
  (params (InteractingPawn :type Fort%20Pawn%20Object%20Reference) (InteractionBeingAttempted :type EInteractionBeingAttempted%20Enum :default "FirstInteraction"))
  ;[1 1 1] Carry
  (bind _returnvalue (FortPawn|PickUpActor InteractingPawn self (Variables|Config|GetPickupDecoDefinition)))
  (if _returnvalue
    (Variables|Default|SetHoldingPawn InteractingPawn)
    (Variables|Replicated|SetbBeingCarried true)
    (Variables|Default|SetLastPickupPosition (Transformation|GetWorldLocation (Variables|Default|GetBaseCollisionCylinder)))
    (Audio|SpawnSoundAttached "/Game/Sounds/Fort_Gadgets/Teleporter/teleporter_pickup_Cue.teleporter_pickup_Cue" (Variables|Default|GetSK_Teleporter))
    (CallFunction|BindWeaponSwap)))

(event Custom|FirstPlaceActivation
  (Variables|BusinessLogic|SetbIntentionalPlacement true)
  (Variables|BusinessLogic|SetbValidDropLocation true)
  ;[1 1 1] Drop
  (BusinessLogic|TriggerRecharge) ; Triggers Recharge for both teleporters
  (CallFunction|UnbindWeaponSwap)
  (Variables|Replicated|SetbBeingCarried false)
  (CallFunction|BindFlooronDestroyed)
  (CallFunction|CheckGroundPosition false)
  (BusinessLogic|UpdateTeleportersActive)
  (BusinessLogic|DestroyifSweepAgainstKillZVolume (Transformation|GetActorLocation)))

(event Custom|FloorDestroyRecovery
  (if (not (Variables|Replicated|GetbBeingCarried))
    (Variables|BusinessLogic|SetbIntentionalPlacement false)
    (Variables|BusinessLogic|SetbValidDropLocation false)
    ;[1 1 1] Drop
    (BusinessLogic|TriggerRecharge) ; Triggers Recharge for both teleporters
    (CallFunction|UnbindWeaponSwap)
    (Variables|Replicated|SetbBeingCarried false)
    (CallFunction|BindFlooronDestroyed)
    (CallFunction|CheckGroundPosition false)
    (BusinessLogic|UpdateTeleportersActive)
    (BusinessLogic|DestroyifSweepAgainstKillZVolume (Transformation|GetActorLocation))))

(event Custom|DelayedDestroyActor
  (Actor|DestroyActor))

(event Custom|BindWeaponSwap
  (bind _holdingpawn (Variables|Default|GetHoldingPawn))
  ;[1 1 1] Player weapon swap
  (if (Utilities|IsValid _holdingpawn :!class /Script/Engine.KismetSystemLibrary)
    (FortPawn|AssignOnWeaponEquippedDelegate _holdingpawn (delegate Custom|OnWeaponEquippedDelegate_Event_0))))

(event EventOnDeathServer
  (params (Damage :type Float%20%28single-precision%29) (DamageTags :type Gameplay%20Tag%20Container%20Structure%20%28by%20ref%29) (Momentum :type Vector :default "0, 0, 0") (HitInfo :type Hit%20Result%20Structure%20%28by%20ref%29) (InstigatedBy :type Controller%20Object%20Reference) (DamageCauser :type Actor%20Object%20Reference) (EffectContext :type Gameplay%20Effect%20Context%20Handle%20Structure))
  (bind _targetteleporter (Variables|BusinessLogic|GetTargetTeleporter))
  ;[1 1 1] Destroyed
  (Collision|SetCollisionEnabled (Variables|Default|GetPortalCollisionCylinder) "NoCollision")
  (Utilities|IsValid _targetteleporter
    (:Is%20Valid
      (Actor|DestroyActor _targetteleporter)
      (Variables|BusinessLogic|SetTargetTeleporter 0))))

(event Custom|UnbindWeaponSwap
  (bind _holdingpawn (Variables|Default|GetHoldingPawn))
  ;[1 1 1] Player weapon swap
  (if (Utilities|IsValid _holdingpawn :!class /Script/Engine.KismetSystemLibrary)
    (FortPawn|UnbindEventfromOnWeaponEquippedDelegate _holdingpawn (EventDispatchers|CreateEvent :!event Custom|OnWeaponEquippedDelegate_Event_0))))

(event Custom|OnWeaponEquippedDelegate_Event_0
  (params (NewWeapon :type Fort%20Weapon%20Object%20Reference) (PrevWeapon :type Fort%20Weapon%20Object%20Reference))
  ;[1 1 1] Player weapon swap
  (Utilities|Casting|CastToFortDecoTool NewWeapon
    (:CastFailed
      (Variables|Replicated|SetbBeingCarried false))))

(event EventDestroyed
  (bind _targetteleporter (Variables|BusinessLogic|GetTargetTeleporter))
  ;[1 1 1] Destroyed
  (Networking|SwitchHasAuthority
    (:Authority
      (Collision|SetCollisionEnabled (Variables|Default|GetPortalCollisionCylinder) "NoCollision")
      (Utilities|IsValid _targetteleporter
        (:Is%20Valid
          (Actor|DestroyActor _targetteleporter)
          (Variables|BusinessLogic|SetTargetTeleporter 0))))))

(event EventOnDeathPlayEffects
  (params (Damage :type Float%20%28single-precision%29) (DamageTags :type Gameplay%20Tag%20Container%20Structure%20%28by%20ref%29) (Momentum :type Vector :default "0, 0, 0") (HitInfo :type Hit%20Result%20Structure%20%28by%20ref%29) (InstigatedBy :type Fort%20Pawn%20Object%20Reference) (DamageCauser :type Actor%20Object%20Reference) (EffectContext :type Gameplay%20Effect%20Context%20Handle%20Structure))
  ;[1 1 1] Destroyed
  (Rendering|SetVisibility (Variables|Default|GetSK_Teleporter) false false)
  (Components|Activation|Deactivate (Variables|Default|GetP_Teleporter_Active_01))
  (Components|Activation|Activate (Variables|Default|GetP_Teleporter_Explo) false)
  (Collision|SetCollisionEnabled (Variables|Default|GetPortalCollisionCylinder) "NoCollision")
  (Audio|Components|Audio|FadeOut (Variables|Default|GetTeleporter_Active_AudioComponent) 0.3 0.0 "Linear" :!class /Script/Engine.AudioComponent))

(event EventBeginPlay
  ;[1 1 1] Begin play
  (Networking|SwitchHasAuthority
    (:Authority
      ;[1 1 0] Only run this on the server, to prevent client location getting out of sync. ReplicateMovement = true, so it's not necessary
      (CallFunction|BindFlooronDestroyed))
    (:Remote
      _))
  (CallFunction|SetSearchSpeed))

(event BuildingActor|EventPlacedByPlayer
  (params (PlayerState :type Fort%20Player%20State%20Zone%20Object%20Reference))
  (BusinessLogic|UpdateTeleportersActive))

(event Custom|DirectionalRayActivate
  ;[1 1 1] Directional Ray
  (Utilities|FlowControl|Delay (Variables|VFX|GetDirectionalRayActivate%20Delay))
  (Rendering|Material|SetScalarParameterValue (Variables|VFX|GetLightray%20MID) "Opacity" 0.0 :!class /Script/Engine.MaterialInstanceDynamic)
  (Development|SetHiddeninGame (Variables|Default|GetSm_Teleporter_LightRay_01) false false)
  (bind (_direction _fade) (Timeline|LightrayFade PlayFromStart)
    (:Update
      (Rendering|Material|SetScalarParameterValue (Variables|VFX|GetLightray%20MID) "Opacity" _fade :!class /Script/Engine.MaterialInstanceDynamic))))

(event Custom|DirectionalRayDeactivate
  ;[1 1 1] Directional Ray
  (Development|SetHiddeninGame (Variables|Default|GetSm_Teleporter_LightRay_01) true false))

(event Custom|OnRechargingChanged
  ;[1 1 1] Animate ring for the recharge
  (Networking|CanExecuteCosmeticEvents
    (:True
      (if (Variables|BusinessLogic|GetbRecharging)
        (Utilities|FlowControl|Delay (Components|Timeline|GetTimelineLength (Variables|B_Teleporter|GetGlowMaterialTransitionOut)))
        (Rendering|Material|SetScalarParameterValue (Variables|VFX|GetTeleporter%20MID%20#0) "Emissive Timeline Animation Falloff Radius" (Variables|VFX|GetRing%20Radius%20Start) :!class /Script/Engine.MaterialInstanceDynamic)
        (Rendering|Material|SetScalarParameterValue (Variables|VFX|GetTeleporter%20MID%20#0) "EmissiveAnimationManualTime" 0.0 :!class /Script/Engine.MaterialInstanceDynamic)
        (Rendering|Material|SetScalarParameterValue (Variables|VFX|GetGlow%20MID) "Transition" -1 :!class /Script/Engine.MaterialInstanceDynamic)
        (Rendering|Material|SetScalarParameterValue (Variables|VFX|GetVortex%20MID) "Erode" -1 :!class /Script/Engine.MaterialInstanceDynamic)
        (Components|Timeline|SetPlayRate (Variables|B_Teleporter|GetRingSpeedUp) (/ 1 (- (Variables|Replicated|GetRechargeTime) (Components|Timeline|GetTimelineLength (Variables|B_Teleporter|GetGlowMaterialTransitionOut)))))
        (Audio|Components|Audio|Play (Variables|Default|Getteleporter_recharge_audiocomponent) 0.0)
        (Utilities|FlowControl|Sequence
          (:then_0
            (bind (_direction _lerp _lerp2) (Timeline|RingSpeedUp PlayFromStart)
              (:Update
                ;;[1 1 1] Speed up the marching lights on the inner ring
                (Rendering|Material|SetScalarParameterValue (Variables|VFX|GetTeleporter%20MID%20#0) "EmissiveAnimationManualTime" (Math|Float|Lerp (Variables|VFX|GetRingChargeStartSpeed) (Variables|VFX|GetRingChargeEndSpeed) _lerp) :!class /Script/Engine.MaterialInstanceDynamic)
                (Rendering|Material|SetScalarParameterValue (Variables|VFX|GetTeleporter%20MID%20#0) "Emissive Timeline Animation Falloff Radius" (Math|Float|Lerp (Variables|VFX|GetRing%20Radius%20Start) (Variables|VFX|GetRing%20Radius%20End) _lerp) :!class /Script/Engine.MaterialInstanceDynamic)
                (Effects|Components|ParticleSystem|SetFloatParameter (Variables|Default|GetP_Teleporter_ReloadRing_01) "Progress" _lerp)
                (Effects|Components|ParticleSystem|SetFloatParameter (Variables|Default|GetP_Teleporter_ReloadRing_01) "TimeSpeed" (Math|Float|Lerp 0.5 8 _lerp2))
                (Effects|Components|ParticleSystem|SetFloatParameter (Variables|Default|GetP_Teleporter_ReloadRingRays_01) "SpawnRaysOrange" (Math|Float|Lerp 1 0.1 (Utilities|MakeSteps _lerp)))
                (Effects|Components|ParticleSystem|SetFloatParameter (Variables|Default|GetP_Teleporter_ReloadRingRays_01) "SpawnRaysBlue" (Math|Float|Lerp -0.1 1 (Utilities|MakeSteps _lerp))))))
          (:then_1
            (Utilities|FlowControl|Delay 0.25)
            (Components|Activation|SetActive (Variables|Default|GetP_Teleporter_ReloadRing_01) true true)
            (bind (_direction_2 _scale _meshscale) (Timeline|ScaleLoadingRings PlayFromStart)
              (:Update
                (Effects|Components|ParticleSystem|SetVectorParameter (Variables|Default|GetP_Teleporter_ReloadRing_01) "Size" (Math|Conversions|ToVector(Float) _scale))
                (Transformation|SetRelativeScale3D (Variables|Default|GetDirectionalArrow) (Math|Vector|Lerp(Vector) "0, 0, 0" (Variables|Config|GetTeleporter%20Size%20Placed) _meshscale)))
              (:Finished
                (Components|Activation|SetActive (Variables|Default|GetP_Teleporter_ReloadRing_01) (Utilities|Enum|Equal(Enum) _direction_2 "Forward") true)))))
        (else
          (Components|Timeline|Stop (Variables|B_Teleporter|GetRingSpeedUp))
          (Rendering|Material|SetScalarParameterValue (Variables|VFX|GetTeleporter%20MID%20#0) "Emissive Timeline Animation Falloff Radius" 0.0 :!class /Script/Engine.MaterialInstanceDynamic)
          (Rendering|Material|SetScalarParameterValue (Variables|VFX|GetTeleporter%20MID%20#0) "EmissiveAnimationSpeed" 0.0 :!class /Script/Engine.MaterialInstanceDynamic)
          (Audio|Components|Audio|FadeOut (Variables|Default|Getteleporter_recharge_audiocomponent) 0.5 0.0 "Linear" :!class /Script/Engine.AudioComponent)
          (Components|Timeline|ReversefromEnd (Variables|B_Teleporter|GetScaleLoadingRings)))))))

(event Custom|OnActiveChanged
  ;[1 1 1] Animate glow material and vortex when activated or deactivated
  (Networking|CanExecuteCosmeticEvents
    (:True
      (if (Variables|Replicated|GetbTeleporterActive)
        (Visualization|SetShowGlow true)
        (bind (_direction _transition _lightlerp _glowscaleup) (Timeline|GlowMaterialTransitionIn PlayFromStart)
          (:Update
            (Rendering|Material|SetScalarParameterValue (Variables|VFX|GetGlow%20MID) "Transition" _transition :!class /Script/Engine.MaterialInstanceDynamic)
            (Rendering|Material|SetScalarParameterValue (Variables|VFX|GetVortex%20MID) "Erode" _transition :!class /Script/Engine.MaterialInstanceDynamic)
            (Transformation|SetRelativeScale3D (Variables|Default|GetGlow) (* _glowscaleup (Variables|Config|GetGlowSizeScale)))
            (bind _light (Variables|Default|GetLight)
              (:then
                (Rendering|Components|Light|SetIntensity _light (Math|Float|Lerp 0.0 (Variables|VFX|GetLightIntensity) _lightlerp)))
              (:else))))
        (else
          (bind (_direction_2 _transition_2 _lightlerp_2 _glowscaleup_2) (Timeline|GlowMaterialTransitionOut PlayFromStart)
            (:Update
              (Rendering|Material|SetScalarParameterValue (Variables|VFX|GetGlow%20MID) "Transition" _transition_2 :!class /Script/Engine.MaterialInstanceDynamic)
              (Rendering|Material|SetScalarParameterValue (Variables|VFX|GetVortex%20MID) "Erode" _transition_2 :!class /Script/Engine.MaterialInstanceDynamic)
              (Transformation|SetRelativeScale3D (Variables|Default|GetGlow) (* _glowscaleup_2 (Variables|Config|GetGlowSizeScale)))
              (bind _light_2 (Variables|Default|GetLight)
                (:then
                  (Rendering|Components|Light|SetIntensity _light_2 (Math|Float|Lerp 0.0 (Variables|VFX|GetLightIntensity) _lightlerp_2)))
                (:else)))
            (:Finished
              (Visualization|SetShowGlow false))))))))

(event Custom|ClientActorsTeleported
  (params (TeleportedActor :type Actor%20Object%20Reference))
  ;[0 0.1944 1] New Post Process (played on the player's screen when exiting the teleporter)
  (Networking|CanExecuteCosmeticEvents
    (:True
      (bind _aspawn (Utilities|Casting|CastToPawn TeleportedActor))
      (if (Pawn|IsLocallyControlled _aspawn)
        (bind (_direction _weight) (Timeline|PostProcessTL PlayFromStart)
          (:Update
            (Variables|/Script/Engine.PostProcessComponent|SetBlendWeight _weight (Variables|Default|GetTeleportPostProcess))))))))

(event Custom|CheckGroundPosition
  (params (bFromGroundDestroyed :type Boolean))
  (Networking|SwitchHasAuthority
    (:Authority
      (CallFunction|CheckifAdjustmentNeeded bFromGroundDestroyed))))

(event Custom|UnbindGroundActorEvents
  (bind _groundactor (Variables|Default|GetGroundActor)
    (:then
      (Health|UnbindEventfromOnBuildingHealthChanged _groundactor (delegate Custom|OnBuildingHealthChanged_Event_0))
      (EventDispatchers|UnbindEventfromOnDied _groundactor (delegate Custom|GroundDied)))
    (:else)))

(event Custom|BindFloorOnDestroyed
  (bind (_start _end) (CallFunction|TraceRange))
  (CallFunction|UnbindGroundActorEvents)
  (bind (_outhit _returnvalue) (Collision|LineTraceByChannel :Start _start :End _end :TraceChannel "TraceTypeQuery10" :bTraceComplex false :DrawDebugType "None" :bIgnoreSelf true :TraceColor "(R=1.000000,G=0.000000,B=0.781043,A=1.000000)" :TraceHitColor "(R=0.000000,G=1.000000,B=0.000000,A=1.000000)" :DrawTime 5.0))
  (if _returnvalue
    (bind _asbuilding_actor (Utilities|Casting|CastToBuildingActor (.HitActor (Collision|BreakHitResult _outhit))))
    (bind _output_get (Variables|Default|SetGroundActor _asbuilding_actor))
    (Health|AssignOnBuildingHealthChanged _output_get (delegate Custom|OnBuildingHealthChanged_Event_0))
    (EventDispatchers|BindEventtoOnDied _output_get (delegate Custom|GroundDied))
    (else
      (CallFunction|SetLocationSafe (Variables|Default|GetLastPickupPosition) false true)
      (Variables|Default|SetGroundActor 0))))

(event Custom|OnBuildingHealthChanged_Event_0
  (bind _groundactor (Variables|Default|GetGroundActor))
  (Utilities|IsValid _groundactor
    (:Is%20Valid
      (if (<= (Health|GetHealth _groundactor) 0.0)
        (CallFunction|CheckGroundPosition true)))))

(event Custom|GroundDied
  (params (DamagedActor :type Actor%20Object%20Reference) (Damage :type Float%20%28single-precision%29) (InstigatedBy :type Controller%20Object%20Reference) (DamageCauser :type Actor%20Object%20Reference) (HitLocation :type Vector) (FHitComponent :type Primitive%20Component%20Object%20Reference) (BoneName :type Name) (Momentum :type Vector))
  (bind _asbuilding_static_mesh (Utilities|Casting|CastToBuildingSMActor DamagedActor)
    (:then
      (bind _returnvalue (Building|GetReplacementDestructionReason _asbuilding_static_mesh))
      (if (Utilities|Enum|Equal(Enum) _returnvalue "BRT_Upgrade")
        ;[1 1 1] An upgrade will change our ground actor, but not our location.  Don't need to perform a full ground check.
        (CallFunction|BindFlooronDestroyed)
        (else
          _)))
    (:CastFailed
      _))
  (CallFunction|CheckGroundPosition true))

(fn BlueprintCanInteract
  (params (InteractingPawn :type Fort%20Pawn%20Object%20Reference) (InteractionBeingAttempted :type EInteractionBeingAttempted%20Enum :default "FirstInteraction") (InteractionType :type TInteractionType%20Enum :default "IT_NoInteraction"))
  (if (Team|OnSameTeam InteractingPawn self)
    (Utilities|IsBeingCarried self
      (:Yes
        _)
      (:No
        (LocalVariables|SetRetVal (not (Variables|BusinessLogic|GetbRecharging))))))
  (return (LocalVariables|GetRetVal)))

(fn BlueprintGetInteractionString
  (params (InteractingPawn :type Fort%20Pawn%20Object%20Reference) (InteractionBeingAttempted :type EInteractionBeingAttempted%20Enum :default "FirstInteraction"))
  (return (Variables|Config|GetInteractText)))

(fn CheckIfAdjustmentNeeded
  (params (bFromGroundDestroyed :type Boolean))
  (bind _returnvalue (Transformation|GetActorLocation))
  (bind _foundteleporter (LocalVariables|GetFoundTeleporter))
  (bind _returnvalue_2 (Transformation|GetActorLocation _foundteleporter))
  (bind _holdingpawn (Variables|Default|GetHoldingPawn))
  (bind _targetteleporter (Variables|BusinessLogic|GetTargetTeleporter))
  (bind _array (Utilities|Array|MakeArray _targetteleporter _holdingpawn))
  (bind _newlocation (LocalVariables|GetNewLocation))
  (LocalVariables|SetbFoundTeleporter false)
  (Utilities|IsValid (Variables|Default|GetHoldingPawn)
    (:Is%20Valid
      ;[1 1 1] figure out if we are overlapping a teleporter that has already been placed
      (if (not bFromGroundDestroyed)
        (bind _outhits (Collision|MultiSphereTraceForObjects (+ (Transformation|GetActorLocation) "0, 0,256.000000") (- (Transformation|GetActorLocation) "0, 0,128.000000") 128 (Utilities|Array|MakeArray "ObjectTypeQuery2") false (Utilities|Array|MakeArray (Variables|Default|GetHoldingPawn))))
        (bind _array_element (Utilities|Array|ForEachLoopwithBreak _outhits)
          (:LoopBody
            (bind _asb_teleporter (Utilities|Casting|CastToB_Teleporter (.HitActor (Collision|BreakHitResult _array_element))))
            (LocalVariables|SetFoundTeleporter _asb_teleporter)
            (LocalVariables|SetbFoundTeleporter true))
          (:Completed
            _))
        (if (LocalVariables|GetbFoundTeleporter)
          ;[1 1 1] Teleporter Overlap detected
          ; Try to find a location near by that works
          (LocalVariables|SetFoundTeleporterLocation _returnvalue_2)
          (bind _output_get (LocalVariables|SetHitDirection (Math|Vector|Normalize (- _returnvalue _returnvalue_2))))
          ;;[1 1 1] start by looking for the direct path away from the found teleporter
          ;; and then rotate around to the 3 other directions looking for a valid location
          (for _index (range 4)
            ;;[1 1 1] Make sure that the original placement position and the goal position can be reached without going through world geo (this makes sure that the goal isn't out of world)
            (bind _output_get_2 (LocalVariables|SetNewLocation (Math|Vector|MakeVector (.x (+ (Math|Vector|RotateVectorAroundAxis (* _output_get 200) (* _index 9e+01) "0, 0,1.000000") (LocalVariables|GetFoundTeleporterLocation))) (.y (+ (Math|Vector|RotateVectorAroundAxis (* _output_get 200) (* _index 9e+01) "0, 0,1.000000") (LocalVariables|GetFoundTeleporterLocation))) (.z _returnvalue))))
            (bind (_outhit _returnvalue_3) (Collision|SphereTraceByChannel (+ (Transformation|GetActorLocation) (Math|Vector|MakeVector 0.0 0.0 (+ 16.0 6.0))) (+ _output_get_2 (Math|Vector|MakeVector 0.0 0.0 (+ 16.0 6.0))) 16.0 "TraceTypeQuery10" false _array))
            (if (not _returnvalue_3)
              ;;[1 1 1] make sure that the new end location is valid
              (bind _valid (CallFunction|CheckifValidSpot _newlocation _array))
              (if _valid
                ;;[1 1 1] we found a good spot, move the tele and return the function
                (CallFunction|SetLocationSafe _newlocation false true)
                (return))))
          (elif (Collision|BreakHitResult _array_element)
            _))
        (else
          _))
      ;[1 1 1] NO teleporter overlap detected,
      ; if we are overlapping something else then stay put
      ; if we are NOT overlapping anything else, find a floor
      (CallFunction|FindNewFloor bFromGroundDestroyed))))

(fn CheckIfValidSpot
  (params (In :type Vector) (Ignore :type Array%20of%20Actor%20Object%20References))
  (bind _outhits (Collision|MultiSphereTraceForObjects (+ In "0, 0,256.000000") (+ In "0, 0,130.000000") 128 (Utilities|Array|MakeArray "ObjectTypeQuery2" "ObjectTypeQuery1") false Ignore))
  (bind _array_element (Utilities|Array|ForEachLoopwithBreak _outhits)
    (:LoopBody
      (if (Math|Boolean|ORBoolean (Utilities|ClassIsChildOf (Utilities|GetClass (.HitActor (Collision|BreakHitResult _array_element))) "/Script/FortniteGame.BuildingFloor") (Utilities|ClassIsChildOf (Utilities|GetClass (.HitActor (Collision|BreakHitResult _array_element))) "/Script/FortniteGame.BuildingProp") (Utilities|ClassIsChildOf (Utilities|GetClass (.HitActor (Collision|BreakHitResult _array_element))) "/Script/FortniteGame.FortStaticMeshActor"))
        (LocalVariables|SetFoundObjectInWay true)))
    (:Completed
      _))
  (return (not (LocalVariables|GetFoundObjectInWay))))

(fn DestroyIfSweepAgainstKillZVolume
  (params (NewLocation :type Vector))
  (bind (_origin _boxextent _sphereradius) (Collision|GetComponentBounds (Variables|Default|GetBaseCollisionCylinder)))
  (Networking|SwitchHasAuthority
    (:Authority
      (Utilities|IsValid (Variables|Default|GetBaseCollisionCylinder)
        (:Is%20Valid
          (if (Class|BuildingFunctionLibrary|SweepAgainstActorTypes self (Transformation|GetActorLocation) NewLocation _sphereradius (Utilities|Array|MakeArray "ObjectTypeQuery2") (Utilities|Array|MakeArray "/Script/Engine.KillZVolume"))
            ;[1 1 0] Prevent interaction
            (Variables|BusinessLogic|SetbValidDropLocation false)
            ;[1 0 0] Delayed destroy.  If we destroy immediately on placement, we fail to play the destruction vfx.
            (Utilities|Time|SetTimerbyEvent (EventDispatchers|CreateEvent :!event DelayedDestroyActor) 0.2)))))))

(fn DestroyRelationship
  (Actor|DestroyActor (Variables|BusinessLogic|GetTargetTeleporter))
  (Actor|DestroyActor))

(fn FindNewFloor
  (params (bFromGroundDestroyed :type Boolean))
  (bind (_start _end) (CallFunction|TraceRange))
  (Utilities|Array|Clear (LocalVariables|GetIgnoreActor))
  (if bFromGroundDestroyed
    _
    (else
      _))
  (bind (_outhit _returnvalue) (Collision|LineTraceByChannel _start _end "TraceTypeQuery10" false (LocalVariables|GetIgnoreActor)))
  (if _returnvalue
    (CallFunction|SetLocationSafe (.ImpactPoint (Collision|BreakHitResult _outhit)) true)
    (CallFunction|BindFlooronDestroyed)
    (if (Variables|GetbFromGroundDestroyed)
      (CallFunction|FloorDestroyRecovery))
    (else
      (CallFunction|DestroyRelationship))))

(fn GateTeleportedActorType
  (params (OverlappingActor :type Actor%20Object%20Reference) (IncludeHusks :type Boolean) (IncludeProjectiles :type Boolean) (CollisionType :type ECollisionChannel%20Enum :default "ECC_WorldStatic"))
  (bind _returnvalue (GameplayTags|HasAnyMatchingGameplayTags(Message) OverlappingActor (Variables|Default|GetBlockedFromTeleportationTags)))
  (if (not _returnvalue)
    (if (Math|Boolean|ORBoolean (Utilities|Enum|Equal(Enum) CollisionType "ECC_Pawn") (and IncludeHusks (Utilities|Enum|Equal(Enum) CollisionType "ECC_GameTraceChannel4")) (and IncludeProjectiles (or (and (Utilities|Enum|Equal(Enum) CollisionType "ECC_GameTraceChannel8") (Variables|DesignRestrictions|GetbFriendlyProjectiles)) (and (Utilities|Enum|Equal(Enum) CollisionType "ECC_GameTraceChannel9") (Variables|DesignRestrictions|GetbEnemyProjectiles)))) (Utilities|Enum|Equal(Enum) CollisionType "ECC_Vehicle"))
      (LocalVariables|SetLocalASC (Ability|GetAbilitySystemComponent OverlappingActor))
      (LocalVariables|SetLocalValidActor true)
      (else
        (LocalVariables|SetLocalValidActor false)))
    (return (LocalVariables|GetLocalASC) (LocalVariables|GetLocalValidActor))
    (else
      _)))

(fn GetArrowRotation
  (return (Math|Rotator|MakeRotator 0.0 0.0 (.yaw (Math|Conversions|RotationFromXVector (- (Variables|Replicated|GetTargetTeleporterLocation) (Transformation|GetActorLocation)))))))

(fn GetDestinationLocation
  (params (TeleportActor :type Actor%20Object%20Reference))
  (return (+ (Math|Vector|MakeVector 0.0 0.0 (Math|Float|Max(Float) (- (.z (Transformation|GetActorLocation TeleportActor)) (.z (Transformation|GetActorLocation))))) (Transformation|GetActorLocation (Variables|BusinessLogic|GetTargetTeleporter)))))

(fn GetDestinationRotation
  (params (TeleportActor :type Actor%20Object%20Reference))
  (return (Math|Rotator|MakeRotator (.roll (Transformation|GetActorRotation TeleportActor)) (.pitch (Transformation|GetActorRotation TeleportActor)) (.yaw (Transformation|GetActorRotation (Variables|BusinessLogic|GetTargetTeleporter))))))

(fn InitializeTeleporter
  (params (NewMaxRange :type Float%20%28double-precision%29) (NewArmTime :type Float%20%28double-precision%29) (TeleporterArrowColor :type Linear%20Color%20Structure) (InPawn :type Fort%20Player%20Pawn%20Object%20Reference) (UseFriendlyProjectiles :type Boolean) (UseEnemyProjectiles :type Boolean) (FasterInteraction :type Boolean))
  (Variables|Replicated|SetMaxTeleportRange NewMaxRange)
  (Variables|Replicated|SetRechargeTime NewArmTime)
  (Variables|VFX|SetTeleporterArrowColor TeleporterArrowColor)
  (Variables|Default|SetHoldingPawn InPawn)
  (Variables|DesignRestrictions|SetbFriendlyProjectiles UseFriendlyProjectiles)
  (Variables|DesignRestrictions|SetbEnemyProjectiles UseEnemyProjectiles)
  (Variables|DesignRestrictions|SetbTeleportProjectiles (or UseFriendlyProjectiles UseEnemyProjectiles))
  (Variables|UpgradeVariables|SetbSetSearchSpeed FasterInteraction))

(fn InternalSetActive
  (params (bActive :type Boolean))
  (if (Math|Boolean|NotEqual(Boolean) bActive (Variables|Replicated|GetbTeleporterActive))
    (Variables|Replicated|SetbTeleporterActive bActive)))

(fn IsAcceptablePositionForPlacement
  (params (InLocation :type Vector%20%28by%20ref%29 :default "0, 0, 0") (InRotation :type Rotator%20%28by%20ref%29 :default "0, 0, 0") (DecoTool :type Fort%20Deco%20Tool%20Object%20Reference) (bIsCDO :type Boolean))
  (if (not bIsCDO)
    (Utilities|IsLocationWithinTeleportRange InLocation
      (:Yes
        _)
      (:No
        (LocalVariables|SetRetVal false)
        (LocalVariables|SetFailReason (Variables|Config|GetTooFarAwayMessage)))))
  (return (LocalVariables|GetFailReason) (LocalVariables|GetRetVal)))

(fn IsCarried
  (bind _returnvalue (Variables|Transformation|GetRootComponent))
  (bind _returnvalue_2 (Transformation|GetAttachParent _returnvalue))
  (if (not (Variables|Replicated|GetbBeingCarried))
    ;[1 1 1] Not being carried from other means, stop loop
    (return false)
    (else
      (Utilities|IsValid _returnvalue_2
        (:Is%20Valid
          (Utilities|IsValid (Components|GetOwner _returnvalue_2)
            (:Is%20Valid
              (return true))
            (:Is%20Not%20Valid
              _)))
        (:Is%20Not%20Valid
          _))
      (return false))))

(fn IsExclusivelyActive
  (if (Variables|BusinessLogic|GetbValidDropLocation)
    (if (not (Variables|BusinessLogic|GetbRecharging))
      (Utilities|IsBeingCarried self
        (:Yes
          _)
        (:No
          (if (not (Variables|Default|GetRemoteCarried))
            (Utilities|IsLocationWithinTeleportRange (Transformation|GetActorLocation)
              (:Yes
                (LocalVariables|SetbRetVal true))
              (:No
                _))))))
    (else
      _))
  (return (LocalVariables|GetbRetVal)))

(fn IsValidActorToTeleport
  (params (ActorToTeleport :type Actor%20Object%20Reference))
  ;[1 1 1] Exists
  (bind _output_get (LocalVariables|SetTeleportActor ActorToTeleport))
  (Utilities|IsValid _output_get
    (:Is%20Valid
      ;[1 1 1] Is Movable
      (if (Utilities|Enum|Equal(Enum) (Variables|/Script/Engine.SceneComponent|GetMobility (Variables|Transformation|GetRootComponent (LocalVariables|GetTeleportActor))) "Movable")
        ;[1 1 1] Isn't Attached
        (Utilities|IsValid (Transformation|GetAttachParent (Variables|Transformation|GetRootComponent (LocalVariables|GetTeleportActor)))
          (:Is%20Valid
            _)
          (:Is%20Not%20Valid
            ;[1 1 1] Isn't FortAttached
            (bind _asbuilding_actor (Utilities|Casting|CastToBuildingActor (LocalVariables|GetTeleportActor))
              (:then
                (bind _returnvalue (Attachment|GetParentBuildingAttachment(Message) _asbuilding_actor))
                (Utilities|IsValid _returnvalue
                  (:Is%20Valid
                    _)
                  (:Is%20Not%20Valid
                    (LocalVariables|SetRetVal true)))
                (return (LocalVariables|GetRetVal)))
              (:CastFailed
                _))))))
    (:Is%20Not%20Valid
      _)))

(fn OnRep_bBeingCarried
  (bind _self self)
  (if (Variables|Replicated|GetbBeingCarried)
    (CallFunction|SetRemoteCarried (Variables|BusinessLogic|GetTargetTeleporter)))
  (Networking|CanExecuteCosmeticEvents
    (:True
      (Visualization|UpdateGlowVisuals)
      (if (Variables|Replicated|GetbBeingCarried)
        (CallFunction|DirectionalRayDeactivate)
        (Rendering|SetVisibility (Variables|Default|GetDirectionalArrow) false false)
        (Components|Activation|SetActive (Variables|Default|GetP_Teleporter_ReloadRing_01) false false)
        (Rendering|SetActorHiddenInGame true)
        (else
          (Transformation|SetActorScale3D _self "1.000000,1.000000,1.000000")
          (Transformation|SetActorRotation _self (Transformation|GetActorRotation _self) false)
          (Components|Activation|Activate (Variables|Default|GetGeneric_Object_Place_AudioComponent) true) ; place teleport sound
          (Rendering|SetActorHiddenInGame false)
          (CallFunction|DirectionalRayActivate)
          (Rendering|SetVisibility (Variables|Default|GetDirectionalArrow) true false)
          (Components|Activation|SetActive (Variables|Default|GetP_Teleporter_ReloadRing_01) (Variables|BusinessLogic|GetbRecharging) true))))))

(fn OnRep_bRecharging
  (CallFunction|OnRechargingChanged))

(fn OnRep_bSetSearchSpeed
  (CallFunction|SetSearchSpeed))

(fn OnRep_bTeleporterActive
  (bind _teleporter_active_audiocomponent (Variables|Default|GetTeleporter_Active_AudioComponent))
  (bind _teleporter_active_audiocomponent_2 (Variables|Default|GetTeleporter_Active_AudioComponent))
  (Networking|CanExecuteCosmeticEvents
    (:True
      (CallFunction|OnActiveChanged)
      (if (Variables|Replicated|GetbTeleporterActive)
        (Audio|Components|Audio|SetSound _teleporter_active_audiocomponent_2 "/Game/Sounds/Fort_Gadgets/Teleporter/teleporter_active_loop_Cue.teleporter_active_loop_Cue")
        (Components|Activation|Activate _teleporter_active_audiocomponent_2 true)
        (else
          (Audio|Components|Audio|SetSound _teleporter_active_audiocomponent "/Game/Sounds/Fort_Gadgets/Teleporter/teleporter_inactive_loop_Cue.teleporter_inactive_loop_Cue")
          (Audio|Components|Audio|Play _teleporter_active_audiocomponent 0.0))))))

(fn OnRep_bValidDropLocation)

(fn OnRep_TeleportCount
  (Networking|CanExecuteCosmeticEvents
    (:True
      (Components|Activation|Activate (Variables|Default|GetTeleport_AudioComponent) true) ; Play teleport Sound
      (Components|Activation|Activate (Variables|Default|GetP_Teleporter_Teleported_01) false)
      ;[1 1 1] Set color of streak in P_Teleporter_Teleported_01
      (Effects|Components|ParticleSystem|SetVectorParameter (Variables|Default|GetP_Teleporter_Teleported_01) "StreakColor" (Math|Conversions|ToVector(LinearColor) (Utilities|GetTeamVFX (Variables|VFX|GetStreak%20Color%20%28Friendly%29) (Variables|VFX|GetStreak%20Color%20%28Enemy%29)))))))

(fn OnRep_TeleporterArrowColor
  (Networking|CanExecuteCosmeticEvents))

(fn RechargeComplete
  (Variables|BusinessLogic|SetbRecharging false)
  (Variables|Default|SetRemoteCarried false)
  (BusinessLogic|UpdateTeleportersActive))

(fn SetLastResortPosition
  (bind (_outhit _returnvalue) (Collision|LineTraceByChannel (Transformation|GetActorLocation (Variables|Default|GetHoldingPawn)) (+ (Transformation|GetActorLocation (Variables|Default|GetHoldingPawn)) "0, 0,-65337.000000") "TraceTypeQuery10" false (Utilities|Array|MakeArray (Variables|Default|GetHoldingPawn))))
  (if _returnvalue
    (CallFunction|SetLocationSafe (.ImpactPoint (Collision|BreakHitResult _outhit)) false true)
    (else
      (Actor|DestroyActor))))

(fn SetLocationSafe
  (params (NewLocation :type Vector) (Sweep :type Boolean) (Teleport :type Boolean))
  ;[1 1 0] Only run this on the server, to prevent client location getting out of sync. ReplicateMovement = true, so it's not necessary.
  (Networking|SwitchHasAuthority
    (:Authority
      (BusinessLogic|DestroyifSweepAgainstKillZVolume NewLocation)
      (Transformation|SetActorLocation NewLocation Sweep Teleport))))

(fn SetSearchSpeed
  (bind _output_get (Variables|SetInteractionSpeed_CurveTable (Utilities|Struct|MakeCurveTableRowHandle "/Game/Balance/SearchSpeed.SearchSpeed" (select (Variables|UpgradeVariables|GetbSetSearchSpeed) (Variables|DesignRestrictions|GetFasterInteraction) (Variables|DesignRestrictions|GetDefaultInteraction)))))
  (Networking|SwitchHasAuthority
    (:Authority
      (Development|PrintString "SERVER"))
    (:Remote
      (Development|PrintString "CLIENT")))
  (Development|PrintString (Utilities|String|ToString(Name) (.RowName (Utilities|Struct|BreakCurveTableRowHandle _output_get)))))

(fn SetShowGlow
  (params (Show :type Boolean))
  (Variables|VFX|SetbShowGlow Show)
  (Visualization|UpdateGlowVisuals))

(fn SetTargetTeleporter
  (params (NewTarget :type B%20Teleporter%20Object%20Reference))
  (Variables|BusinessLogic|SetTargetTeleporter NewTarget))

(fn TeleportToTarget
  (params (ActorToTeleport :type Actor%20Object%20Reference))
  (LocalVariables|SetTeleportingActor ActorToTeleport)
  (if (Variables|BusinessLogic|GetbTeleporting)
    _
    (elif (Variables|Replicated|GetbTeleporterActive)
      (bind _validactor (BusinessLogic|IsValidActortoTeleport (LocalVariables|GetTeleportingActor)))
      (if _validactor
        (Utilities|IsValid (Variables|BusinessLogic|GetTargetTeleporter)
          (:Is%20Valid
            (Variables|BusinessLogic|SetbTeleporting true (Variables|BusinessLogic|GetTargetTeleporter))
            (Variables|BusinessLogic|SetbTeleporting true)
            (Transformation|Teleport (LocalVariables|GetTeleportingActor) (CallFunction|GetDestinationLocation (LocalVariables|GetTeleportingActor)) (CallFunction|GetDestinationRotation (LocalVariables|GetTeleportingActor)))
            (Utilities|FlowControl|Sequence
              (:then_0
                ;[1 1 1] Destination Velocity
                (for _array_element (Actor|GetComponentsByClass (LocalVariables|GetTeleportingActor) "/Script/Engine.MovementComponent")
                  (Variables|/Script/Engine.MovementComponent|SetVelocity (* (Math|Rotator|GetRotationXVector (Math|Rotator|MakeRotator (.roll (Math|Conversions|RotationFromXVector (Variables|/Script/Engine.MovementComponent|GetVelocity _array_element))) (.pitch (Math|Conversions|RotationFromXVector (Variables|/Script/Engine.MovementComponent|GetVelocity _array_element))) (.yaw (Transformation|GetActorRotation (Variables|BusinessLogic|GetTargetTeleporter))))) (Math|Vector|VectorLength (Variables|/Script/Engine.MovementComponent|GetVelocity _array_element))) _array_element)))
              (:then_1
                (CallFunction|ClientActorsTeleported (LocalVariables|GetTeleportingActor))
                (Variables|Replicated|SetTeleportCount (+ (Variables|Replicated|GetTeleportCount) 1))
                (Variables|Replicated|SetTeleportCount (+ (Variables|Replicated|GetTeleportCount) 1) (Variables|BusinessLogic|GetTargetTeleporter))
                (Variables|BusinessLogic|SetbTeleporting false)))))))))

(fn TraceRange
  (return (+ (Transformation|GetActorLocation) "0, 0,128.000000") (+ (Transformation|GetActorLocation) "0, 0,-65537.000000")))

(fn TriggerRecharge
  (BusinessLogic|TriggerRechargeInternal)
  (BusinessLogic|TriggerRechargeInternal (Variables|BusinessLogic|GetTargetTeleporter)))

(fn TriggerRechargeInternal
  (Utilities|Time|SetTimerbyFunctionName 0 "RechargeComplete" (Variables|Replicated|GetRechargeTime))
  (Variables|BusinessLogic|SetbRecharging true))

(fn UpdateGlowVisuals
  (bind _bbeingcarried (Variables|Replicated|GetbBeingCarried))
  (bind _returnvalue (not _bbeingcarried))
  (bind _bshowglow (Variables|VFX|GetbShowGlow))
  (bind _returnvalue_2 (and _returnvalue _bshowglow))
  (bind _returnvalue_3 (not _returnvalue_2))
  (Networking|CanExecuteCosmeticEvents
    (:True
      (Development|SetHiddeninGame (Variables|Default|GetGlow) _returnvalue_3 false)
      (Development|SetHiddeninGame (Variables|Default|GetVortex) _returnvalue_3 false)
      (Components|Activation|SetActive (Variables|Default|GetP_Teleporter_Active_01) (not _returnvalue_3) false)
      (bind _light (Variables|Default|GetLight)
        (:then
          (Development|SetHiddeninGame _light _returnvalue_3 false))
        (:else)))))

(fn UpdateTargetTeleporterLocation
  (Variables|Replicated|SetTargetTeleporterLocation (Transformation|GetActorLocation (Variables|BusinessLogic|GetTargetTeleporter))))

(fn UpdateTeleportersActive
  (Utilities|IsValid (Variables|BusinessLogic|GetTargetTeleporter)
    (:Is%20Valid
      (Visualization|UpdateTargetTeleporterLocation)
      (Visualization|UpdateTargetTeleporterLocation (Variables|BusinessLogic|GetTargetTeleporter))
      (bind _isactive (BusinessLogic|IsExclusivelyActive (Variables|BusinessLogic|GetTargetTeleporter)))
      (bind _isactive_2 (BusinessLogic|IsExclusivelyActive))
      (BusinessLogic|InternalSetActive (and _isactive_2 _isactive))
      (BusinessLogic|InternalSetActive (Variables|BusinessLogic|GetTargetTeleporter) (and _isactive_2 _isactive)))
    (:Is%20Not%20Valid
      (BusinessLogic|InternalSetActive false))))

(fn ConstructionScript
  (bind _light (Variables|Default|GetLight))
  (Networking|CanExecuteCosmeticEvents
    (:True
      (Utilities|EffectsQualityCheck 0
        (:Cinematic
          _)
        (:Epic
          _)
        (:High
          _)
        (:Medium
          (Components|DestroyComponent _light)
          (bind _returnvalue (Rendering|Material|CreateDynamicMaterialInstance (Variables|Default|GetSK_Teleporter) 0 "/Game/Gadgets/Teleporter/Material/Fortress_M_TeleportationDevice_Inst.Fortress_M_TeleportationDevice_Inst" "None" :!class /Script/Engine.PrimitiveComponent))
          (Variables|VFX|SetTeleporter%20MID%20#0 _returnvalue)
          (Rendering|Material|SetScalarParameterValue (Variables|VFX|GetTeleporter%20MID%20#0) "Emissive Timeline Animation Falloff Radius" 0.0 :!class /Script/Engine.MaterialInstanceDynamic)
          (Rendering|Material|SetScalarParameterValue (Variables|VFX|GetTeleporter%20MID%20#0) "EmissiveAnimationSpeed" 0.0 :!class /Script/Engine.MaterialInstanceDynamic)
          (bind _returnvalue_2 (Rendering|Material|CreateDynamicMaterialInstance (Variables|Default|GetSK_Teleporter) 1 "/Game/Gadgets/Assets/Teleporter/Materials/M_Teleporter_Glow_NoRefraction.M_Teleporter_Glow_NoRefraction" "None" :!class /Script/Engine.PrimitiveComponent))
          (Variables|VFX|SetTeleporter%20MID%20#1 _returnvalue_2)
          (bind _returnvalue_3 (Rendering|Material|CreateDynamicMaterialInstance (Variables|Default|GetGlow) 0 "/Game/Gadgets/Assets/Teleporter/Materials/M_Teleporter_Glow_NoRefraction.M_Teleporter_Glow_NoRefraction" "None" :!class /Script/Engine.PrimitiveComponent))
          (Variables|VFX|SetGlow%20MID _returnvalue_3)
          (bind _returnvalue_4 (Rendering|Material|CreateDynamicMaterialInstance (Variables|Default|GetVortex) 0 "/Game/Gadgets/Assets/Teleporter/Materials/M_TeleporterVortex_01.M_TeleporterVortex_01" "None" :!class /Script/Engine.PrimitiveComponent))
          (Variables|VFX|SetVortex%20MID _returnvalue_4)
          (Transformation|SetRelativeScale3D (Variables|Default|GetGlow) (Variables|Config|GetGlowSizeScale))
          (Transformation|SetRelativeScale3D (Variables|Default|GetSm_Teleporter_LightRay_01) (* (Variables|Config|GetTeleporter%20Size%20Placed) (Variables|VFX|GetLightray%20Size%20Length)))
          (bind _returnvalue_5 (Rendering|Material|CreateDynamicMaterialInstance (Variables|Default|GetSm_Teleporter_LightRay_01) 0 0 "None" :!class /Script/Engine.PrimitiveComponent))
          (Variables|VFX|SetLightray%20MID _returnvalue_5)
          (Transformation|SetRelativeScale3D (Variables|Default|GetDirectionalArrow) (Variables|Config|GetTeleporter%20Size%20Placed))
          (bind _returnvalue_6 (Rendering|Material|CreateDynamicMaterialInstance (Variables|Default|GetDirectionalArrow) 0 "/Game/Gadgets/Assets/Teleporter/Materials/M_TeleporterRing_02.M_TeleporterRing_02" "None" :!class /Script/Engine.PrimitiveComponent))
          (Variables|VFX|SetDirectional%20Arrow%20MID _returnvalue_6))
        (:Low
          _))
      (Variables|VFX|SetLightIntensity (Variables|/Script/Engine.LightComponentBase|GetIntensity _light)))))
